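# Niagara system for the AnvilAlarmMoth emote that produces beat-synced coloured spotlights and small drifting particles.
# Its single emitter bursts two spawn groups when User.BeatSyncTest changes, gated by system age: a few large lights and smaller curl-noise particles.
# The beat value also selects light count, lifetime and per-partition alpha masking.
# Colours vary by partition, time of day and front-end flags, feeding mesh, sprite and light renderers through fade and flicker curves.
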NiagaraSystem /BRCosmetics/Animation/Game/MainPlayer/Emotes/AnvilAlarmMoth/FX/NS_Emote_AnvilAlarmMoth_Lights
dependencies {
 usedDataInterfaces [&NiagaraDataInterfaceArrayColor, &NiagaraDataInterfaceArrayFloat, &NiagaraDataInterfaceCurve]
 usedDynamicInputs []
 usedModules [&SystemState]
}
emitters {Spotlight001 {
  emitterUpdateScript [
   &EmitterState(Life Cycle Mode=::NewEnumerator1, Loop Behavior=::NewEnumerator1, Loop Delay=0.536, Loop Duration Mode=::NewEnumerator1, UseLoopDelay=false)
   /Game/Effects/Niagara/Modules/Logic/UpdateValueStateChange(InputIntParam:NiagaraInt32=&MakeIntFromFloat(Float=$User.BeatSyncTest, Conversion Mode=::NewEnumerator1))
   /Niagara/Modules/Emitter/SpawnByTrigger(
    Spawn Count:NiagaraInt32=&MergeHelper_IntSwitch(
     0=2
     1=2
     2=3
     3=?
     4=?
     5=?
     Index=&MakeIntFromFloat(Float=$User.BeatSyncTest, Conversion Mode=::NewEnumerator1)
     Num=::NewEnumerator1
    )
    Trigger:NiagaraBool=&Boolean_AndOperation(Bool1=$Emitter.UpdateValueStateChange.Changed, Bool2=&SetBoolByFloatComparison(A=$Engine.System.Age, B=0.5))
    Trigger Interval=4
    TriggerIntervalSwitch=false
   )
   &SpawnByTrigger(
    Spawn Count:NiagaraInt32=&MergeHelper_IntSwitch(
     0=15
     1=15
     2=15
     3=?
     4=?
     5=?
     Index=&MakeIntFromFloat(Float=$User.BeatSyncTest, Conversion Mode=::NewEnumerator1)
     Num=::NewEnumerator1
    )
    Spawn Group=1
    Trigger:NiagaraBool=&Boolean_AndOperation(Bool1=$Emitter.UpdateValueStateChange.Changed, Bool2=&SetBoolByFloatComparison(A=$Engine.System.Age, B=0.5))
    Trigger Interval=4
    TriggerIntervalSwitch=false
   ) {module=SpawnByTrigger001}
  ]
  particleSpawnScript [
   /SetVariables(Particles.VisibilityTag=&SpawnGroup()) {isSetParameters=true}
   &PartitionParticles()
   /SetVariables(Particles.LightColor=&FNTimeOfDayColor(
     Night Color=&MakeCustomLinearColorFromBool(
      Bool=&Boolean_OrOperation(Bool1=$User.bIsFrontEnd, Bool2=$User.bIsFrontEndPreview)
      True Color=&SelectLinearColorFromArray(
       Array Sampling Mode=::NewEnumerator0
       Direct Array Index=$Particles.PartitionParticles.Partition
       Color Selection Array=DataInterface() {propertyValues {ColorData [
          <rgba 1 0.5514 0 0.275>
          <rgba 0 0.8 0.0916 0.275>
          <rgba 0.0622 0.174 0.9583 0.275>
          <rgba 0.6092 0.0515 0.8 0.275>
          <rgba 0 0.8874 0.9375 0.275>
         ]}}
       Random Seed=0
      )
      False Color=&SelectLinearColorFromArray(
       Array Sampling Mode=::NewEnumerator0
       Direct Array Index=$Particles.PartitionParticles.Partition
       Color Selection Array=DataInterface() {propertyValues {ColorData [
          <rgba 2.713 1.044 0 0.113>
          <rgba 0 2.6415 2.713 0.0696>
          <rgba 0 2.713 0.5622 0.113>
          <rgba 2.713 0 2.1867 0.113>
          <rgba 0.1868 0.33 2.713 0.113>
         ]}}
       Random Seed=0
      )
     )
     Morning Color=<rgba 0.5 0.4132 0.25 0.6>
     Day Color=&MakeCustomLinearColorFromBool(
      Bool=&Boolean_OrOperation(Bool1=$User.bIsFrontEnd, Bool2=$User.bIsFrontEndPreview)
      True Color=&SelectLinearColorFromArray(
       Array Sampling Mode=::NewEnumerator0
       Direct Array Index=$Particles.PartitionParticles.Partition
       Color Selection Array=DataInterface() {propertyValues {ColorData [
          <rgba 0.8 0.3719 0 0.275>
          <rgba 0 0.8 0.0916 0.275>
          <rgba 0.0622 0.174 0.9583 0.275>
          <rgba 0.6092 0.0515 0.8 0.275>
          <rgba 0 0.8874 0.9375 0.275>
         ]}}
       Random Seed=0
      )
      False Color=&SelectLinearColorFromArray(
       Array Sampling Mode=::NewEnumerator0
       Direct Array Index=$Particles.PartitionParticles.Partition
       Color Selection Array=DataInterface() {propertyValues {ColorData [
          <rgba 2.713 1.539 0 0.113>
          <rgba 0 2.6415 2.713 0.0696>
          <rgba 0 2.713 0.5622 0.113>
          <rgba 2.713 0 2.1867 0.113>
          <rgba 0.1868 0.33 2.713 0.113>
         ]}}
       Random Seed=0
      )
     )
     Evening Color=<rgba 0.5 0.4132 0.25 0.6>
    )) {isSetParameters=true}
   &InitializeParticle(
    Color=$Particles.LightColor
    Lifetime=&MergeHelper_FloatSwitch(
     0=&SelectFloatFromArray(
      Direct Array Index=&MakeIntFromFloat(Float=$User.BeatSyncTest)
      Float Selection Array=DataInterface() {propertyValues {FloatData [0.4,0.4,1]}}
     )
     1=&RandomRangeFloat(Minimum=0.6, Maximum=1.5)
     2=?
     3=?
     Index=$Emitter.SpawnGroup
     Num=::NewEnumerator0
    )
    Mesh Scale=<0.9 0.9 1.3174>
    Mesh Scale Mode=::NewEnumerator4
    Position Offset=<0 0 5>
    Sprite Size Mode=::NewEnumerator3
    Uniform Sprite Size=&MergeHelper_FloatSwitch(
     0=350
     1=&RandomRangeFloat(Minimum=5, Maximum=15)
     2=?
     3=?
     Index=&SpawnGroup()
     Num=::NewEnumerator0
    )
    UsePositionOffset=true
    Write Material Random=true
   )
   &InitialMeshOrientation(Mesh Orientation Mode=::NewEnumerator7, Use Rotation Vector=true)
   /SetVariables(Particles.GlowCameraOffset=33) {isSetParameters=true}
   &DynamicMaterialParameters(
    Index 0 Param 1=0.2
    Module.Index0Float0Write=true
    Module.Index0Float1Write=true
    Module.Index0Float2Write=true
    Module.Index0Float3Write=true
   )
   &ShapeLocation(
    Hemisphere Distribution=<1 0.108>
    MaskSpawnGroup=true
    Non Uniform Scale=<1.3 1.3 1>
    Offset Mode=::NewEnumerator0
    Override Seed=false
    Sphere Radius=200
    Sphere Surface Distribution=1
    UseOwnerScale=true
   )
   &ShapeLocation(
    MaskSpawnGroup=true
    Non Uniform Scale=<1 1 2>
    Offset=<0 0 40>
    Offset Mode=::NewEnumerator0
    Override Seed=false
    Spawn Group Mask=1
    Sphere Radius=48.0805
    UseOwnerScale=true
   ) {module=ShapeLocation001}
  ]
  particleUpdateScript [
   &ParticleState()
   &ScaleSpriteSize(Uniform Curve Sprite Scale=DataInterface() {propertyValues {Curve {defaultValue=0, keys [{tangentMode=RCTM_User, time=0.5037, value=1}, {tangentMode=RCTM_User, time=1, value=0}]}}})
   &CurlNoiseForce(
    Bias Noise=false
    Noise Frequency=131.1277
    Noise Strength=&MergeHelper_FloatSwitch(0=0, 1=1301.5441, 2=?, 3=?, Index=$Emitter.SpawnGroup, Num=::NewEnumerator0)
    Pan Noise=false
   )
   /SetVariables(
    Particles.Amount:NiagaraFloat=&Multiply_Float(A=$Emitter.Age, B=0)
    Particles.Amount2:NiagaraFloat=&Add_Float(A=&Multiply_Float(A=$Emitter.Age, B=0), B=0.5)
   ) {isSetParameters=true}
   &ScaleColor(
    Scale Alpha=&FloatFromCurve(
     FloatCurve=DataInterface() {propertyValues {Curve {
        defaultValue=1.0155
        keys [
         {arriveTangent=0.1648, interpMode=RCIM_Cubic, leaveTangent=0.1648, tangentMode=RCTM_User, time=0, value=1}
         {tangentMode=RCTM_User, time=0.8488, value=0}
        ]
       }}}
     Scale Curve=&Multiply_Float(
      A=1
      B=&MergeHelper_FloatSwitch(
       0=&MergeHelper_FloatSwitch(0=1, 1=0, 2=1, 3=?, Index=&MakeIntFromFloat(Float=$User.BeatSyncTest), Num=::NewEnumerator1)
       1=&MergeHelper_FloatSwitch(0=0, 1=1, 2=1, 3=?, Index=&MakeIntFromFloat(Float=$User.BeatSyncTest), Num=::NewEnumerator1)
       2=?
       3=?
       Index=$Particles.PartitionParticles.Partition
       Num=::NewEnumerator0
      )
     )
    )
    ScaleA=true
    ScaleRGB=false
   )
   /SetVariables(Particles.Quat:Quat4f=&OrientMeshToVector(Look At Direction=$Particles.ShapeLocation.ShapeNormal, Reference Axis Vector=<1 0 0>)) {isSetParameters=true}
   &FN_Light_Attributes(
    Color=$Particles.LightColor
    Create Color Attribute=true
    Create Position Attribute=true
    Create Radius Attribute=true
    Position=&AddVectorToPosition(Position=$Particles.Position, Vector=<0 0 50>)
    Radius=150
    Scale Color=&FloatFromCurve(FloatCurve=DataInterface() {propertyValues {Curve {
        defaultValue=0
        keys [
         {interpMode=RCIM_Cubic, tangentMode=RCTM_User, time=0, value=0}
         {interpMode=RCIM_Cubic, tangentMode=RCTM_User, time=0.239, value=1}
         {arriveTangent=-0.4328, interpMode=RCIM_Cubic, leaveTangent=-0.4328, tangentMode=RCTM_User, time=0.7753, value=1}
         {tangentMode=RCTM_User, time=1, value=0}
        ]
       }}})
   )
   /SetVariables(Particles.Component Light Intensity=200.8385) {isSetParameters=true}
   /SetVariables(
    Particles.BokehColor:LinearColor=&Multiply_LinearColorByFloat(
     Linear Color=$Particles.Color
     Float=&FloatFromCurve(
      FloatCurve=DataInterface() {propertyValues {Curve {keys [<at 0.2599 0.5>,<at 1 0>]}}}
      Scale Curve=&Waveform(
       "Waveform [1]"=::NewEnumerator0
       "[1] Frequency"=8.2924
       "[1] Amplitude Min/Max"=<0.5 1>
       "[1] Scale"=&RandomRangeFloat(Minimum=0.2, Maximum=1)
       "[1] Bias"=1
      )
     )
    )
    Particles.LightRayColor:LinearColor=&Multiply_LinearColorByFloat(
     Linear Color=&Multiply_LinearColorByFloat(Linear Color=$Particles.Color, Float=2)
     Float=&FloatFromCurve(FloatCurve=DataInterface() {propertyValues {Curve {
         defaultValue=1.0155
         keys [
          {arriveTangent=-0.5262, interpMode=RCIM_Cubic, leaveTangent=-0.5262, tangentMode=RCTM_User, time=0, value=0}
          {arriveTangent=0.1825, interpMode=RCIM_Cubic, leaveTangent=0.1825, tangentMode=RCTM_User, time=0.2422, value=1}
          {arriveTangent=0.1648, interpMode=RCIM_Cubic, leaveTangent=0.1648, tangentMode=RCTM_User, time=0.6018, value=1}
          {tangentMode=RCTM_User, time=0.8488, value=0}
         ]
        }}})
    )
    Particles.Lightcolortest:LinearColor=&Multiply_LinearColorByFloat(
     Linear Color=$Particles.LightColor
     Float=&FloatFromCurve(FloatCurve=DataInterface() {propertyValues {Curve {
         defaultValue=1.0155
         keys [
          {arriveTangent=-0.5262, interpMode=RCIM_Cubic, leaveTangent=-0.5262, tangentMode=RCTM_User, time=0, value=0}
          {arriveTangent=0.1825, interpMode=RCIM_Cubic, leaveTangent=0.1825, tangentMode=RCTM_User, time=0.2422, value=1}
          {arriveTangent=0.1648, interpMode=RCIM_Cubic, leaveTangent=0.1648, tangentMode=RCTM_User, time=0.7706, value=1}
          {tangentMode=RCTM_User, time=1, value=0}
         ]
        }}})
    )
    Particles.Spotlight:Vector3f=<0 0 -150>
   ) {isSetParameters=true}
   &SolveForcesAndVelocity(Clamp Velocity=false, Limit Acceleration=false)
   /SetVariables(Particles.FlareColor=&Multiply_LinearColorByFloat(
     Linear Color=&Multiply_LinearColorByFloat(Linear Color=$Particles.Color, Float=1)
     Float=&FloatFromCurve(
      FloatCurve=DataInterface() {propertyValues {Curve {keys [<at 0.2599 0.5>,<at 1 0>]}}}
      Scale Curve=&Waveform(
       "Waveform [1]"=::NewEnumerator0
       "[1] Frequency"=8.2924
       "[1] Amplitude Min/Max"=<0.5 1>
       "[1] Scale"=&RandomRangeFloat(Minimum=0.2, Maximum=1)
       "[1] Bias"=1
      )
     )
    )) {isSetParameters=true}
  ]
  rendererClasses [
   &NiagaraMeshRendererProperties
   &NiagaraLightRendererProperties
   &NiagaraSpriteRendererProperties
  ]
  renderers [
   {rendererClass=&NiagaraMeshRendererProperties, rendererIndex=0}
   {rendererClass=&NiagaraMeshRendererProperties, rendererIndex=1}
   {rendererClass=&NiagaraMeshRendererProperties, rendererIndex=2}
   {rendererClass=&NiagaraLightRendererProperties, rendererIndex=3}
   {rendererClass=&NiagaraSpriteRendererProperties, rendererIndex=4}
   {rendererClass=&NiagaraSpriteRendererProperties, rendererIndex=5}
  ]
 }}
summary {
 systemName=NS_Emote_AnvilAlarmMoth_Lights
 emitters [{
   emitterName=Spotlight001
   rendererClasses [
    &NiagaraMeshRendererProperties
    &NiagaraLightRendererProperties
    &NiagaraSpriteRendererProperties
   ]
  }]
 userVariables [
  User.BeatSyncTest {type=NiagaraFloat, defaultValue=1, description=""}
  User.bIsFrontEnd {type=NiagaraBool}
  User.bIsFrontEndPreview {type=NiagaraBool}
  User.OwnerLocalVelocity {type=Vector3f, description="", defaultValue {struct=Vector3f, value=<0 0 0>}}
 ]
}
